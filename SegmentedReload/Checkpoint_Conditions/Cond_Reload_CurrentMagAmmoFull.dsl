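; Segmented reload checkpoint condition that reports whether the weapon's current magazine is full.
; EvaluateCondition returns true when the weapon's magazine ammo count is at or above its bullets-per-clip value.
; It reads only the weapon actor; the reload state data passed in goes unused.
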(fn EvaluateCondition
  (params (ReloadStateData :type Segmented%20Reload%20State%20Data%20Structure%20%28by%20ref%29) (WeaponActor :type Fort%20Weapon%20Object%20Reference))
  (return (>= (FortWeapon|GetMagazineAmmoCount WeaponActor) (FortWeapon|GetBulletsPerClip WeaponActor))))
