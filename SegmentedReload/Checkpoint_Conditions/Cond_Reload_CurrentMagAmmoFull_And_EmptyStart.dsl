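; Segmented reload checkpoint condition that passes when a reload started from an empty magazine and the magazine is now full.
; EvaluateCondition reads the initial reload ammo from the segmented reload state data and the current magazine count and clip size from the Fort weapon actor.
; It returns true when the starting ammo is zero and the current count is at least the clip size.
; A debug Print String outputs current, max and start values on each evaluation.
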(fn EvaluateCondition
  (params (ReloadStateData :type Segmented%20Reload%20State%20Data%20Structure%20%28by%20ref%29) (WeaponActor :type Fort%20Weapon%20Object%20Reference))
  (bind _initialreloadammo (Utilities|Struct|BreakSegmentedReloadStateData ReloadStateData))
  (bind _returnvalue (FortWeapon|GetBulletsPerClip WeaponActor))
  (bind _returnvalue_2 (FortWeapon|GetLocalMagazineAmmoCount WeaponActor))
  (Development|PrintString 0 (Utilities|String|Append "Current: " (Utilities|String|ToString(Integer) _returnvalue_2) "Max:  " (Utilities|String|ToString(Integer) _returnvalue) " Start: " (Utilities|String|ToString(Integer) _initialreloadammo)) true true "(R=1.000000,G=0.129893,B=0.394062,A=1.000000)" 5.0)
  (return (and (== _initialreloadammo 0) (>= _returnvalue_2 _returnvalue))))
